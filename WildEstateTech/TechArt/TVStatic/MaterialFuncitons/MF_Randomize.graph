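# MF_Randomize is a material function in the TV static tech art folder that hashes a 2D input, UVs by default, into pseudo-random values.
# "Random 2 Vector" is a scalar: the fraction of the sine of a dot product, scaled by RandomSeed.
# "Random 3 Vector" takes the sine of three dot products over reshuffled input components and constant vectors, without seed or fraction, so it stays signed.
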MaterialFunction /WildEstateTech/TechArt/TVStatic/MaterialFuncitons/MF_Randomize
graph {nodes [
  n1 = FunctionInput(Preview=TextureCoordinate()) {bUsePreviewValueAsDefault=true, inputName=2 Vector, inputType=FunctionInput_Vector2}
  n2 = ComponentMask(None=@n1) {r=true}
  n3 = ComponentMask(None=@n1) {g=true}
  n4 = DotProduct(
   A=MaterialFunctionCall(R=@n2, G=@n3, B=@n2) {materialFunction=&AppendMany}.RGB
   B=Constant3Vector() {constant=<rgba 127 312 74.7 0>}
  )
  n5 = DotProduct(
   A=MaterialFunctionCall(R=@n3, G=@n2, B=@n2) {materialFunction=&AppendMany}.RGB
   B=Constant3Vector() {constant=<rgba 270 183 246 0>}
  )
  n6 = DotProduct(
   A=MaterialFunctionCall(R=@n2, G=@n3, B=@n3) {materialFunction=&AppendMany}.RGB
   B=Constant3Vector() {constant=<rgba 114 272 125 0>}
  )
  FunctionOutput(None=Sine(None=MaterialFunctionCall(R=@n4, G=@n5, B=@n6) {materialFunction=&AppendMany}.RGB)) {outputName=Random 3 Vector, sortPriority=1}
  n7 = Sine(None=DotProduct(A=@n1, B=Constant2Vector() {g=3.2257, r=1.3322}))
  n8 = FunctionInput(Preview=Constant() {r=43758.5469}) {bUsePreviewValueAsDefault=true, inputName=RandomSeed, inputType=FunctionInput_Scalar, sortPriority=4, previewValue=<7 0 0 1>}
  FunctionOutput(None=Frac(None=Multiply(A=@n7, B=@n8) {constB=43758.5469})) {outputName=Random 2 Vector}
 ]}
